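# Row widget for a Foley Clipboard audio tool, with a label, volume and pitch multiplier boxes, and a Copy button.
# InitFromConfig stores key, label, kind, tag and hotkey, hiding both boxes for footstep rows and setting them to 1.0 otherwise.
# Copy, or a matching key passed to TryHandleHotkey, broadcasts OnStampClicked with the row key and raw volume/pitch text.
# ResetToDefaults restores both boxes to 1.0.
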WidgetBlueprint /FortTechAudioTools/Python/FoleyClipboard/WBP_FoleyClipboardRow
graphs [
 EventGraph {nodes [
   "On Clicked (CopyButton)"() {is=title K2Node_ComponentBoundEvent}
   > Call On Stamp Clicked!(
    execute:Exec
    self:WBP Foley Clipboard Row Object Reference=Self-Reference()
    Key:Name=Get RowKey() -> (RowKey:Name=None)
    VolText:Text="GetText (Text Box)"!(self:Text Box Object Reference=Get VInput() -> (VInput:Text Box Object Reference)) -> (ReturnValue:Text)
    PitchText:Text="GetText (Text Box)"(self=Get PInput() -> (PInput:Text Box Object Reference))
   ) -> (then:Exec) {is=title K2Node_CallDelegate}
   Event PreConstruct()
   Event Construct()
   Event Tick()
  ]}
 InitFromConfig {nodes [
   n1 = InitFromConfig!() -> (then, InKey:Name, InLabel:Text, InKind:Name, InTag:Name, InHotkey:Name) {is=title K2Node_FunctionEntry}
   > Set RowKey(RowKey:Name=@n1) -> (Output_Get:Name=None)
   > Set RowLabel(RowLabel:Text=@n1.InLabel) -> (Output_Get:Text)
   > Set RowKind(RowKind:Name=@n1.InKind) -> (Output_Get:Name=None)
   > Set RowTag(RowTag:Name=@n1.InTag) -> (Output_Get:Name=None)
   > Set RowHotkey(RowHotkey:Name=@n1.InHotkey) -> (Output_Get:Name=None)
   > "SetText (Text)"(self=Get LabelText() -> (LabelText:Text Object Reference), InText=@n1.InLabel)
   > n2 = Branch(Condition="Equal (Name)"(A=@n1.InKind, B:String=MakeLiteralString(Value=footstep)))
   > SetVisibility(self=Get VInput(), InVisibility=Hidden)
   > SetVisibility(self=Get PInput(), InVisibility=Hidden)
   "SetText (Text Box)"!(execute=@n2.else, self:Text Box Object Reference=Get VInput(), InText:Text="LOCTEXT(\"1.0\")") -> (then)
   > "SetText (Text Box)"(self=Get PInput(), InText="LOCTEXT(\"1.0\")")
  ]}
 OnStampClicked {nodes [OnStampClicked!() -> (then, Key:Name, VolText:Text, PitchText:Text) {is=title K2Node_FunctionEntry}]}
 ResetToDefaults {nodes [
   ResetToDefaults() {is=title K2Node_FunctionEntry}
   > "SetText (Text Box)"(self=Get VInput(), InText="LOCTEXT(\"1.0\")")
   > "SetText (Text Box)"(self=Get PInput(), InText="LOCTEXT(\"1.0\")")
  ]}
 TryHandleHotkey {nodes [
   n1 = TryHandleHotkey!() -> (then, InKeyString:String) {is=title K2Node_FunctionEntry}
   > n2 = Branch(Condition="Equal (Name)"(A=Get RowHotkey() -> (RowHotkey:Name=None), B:Name=String To Name(InString=@n1)))
   > Call On Stamp Clicked(
    self=Self-Reference()
    Key=Get RowKey() -> (RowKey=None)
    VolText="GetText (Text Box)"(self=Get VInput())
    PitchText="GetText (Text Box)"(self=Get PInput())
   )
   > Return Node!(execute, bHandled:bool=true)
   Return Node(execute=@n2.else, bHandled=false)
  ]}
]
variables [
 OnStampClicked {}
 RowHotkey {type=string, default=None}
 RowKey {type=string, default=None}
 RowKind {type=string, default=None}
 RowLabel {type=string, default=""}
 RowTag {type=string, default=None}
]
widgets {
 info {inheritedWidgetCount=0, parentClass=&UserWidget, rootWidgetClass=&HorizontalBox, widgetCount=12}
 widgets [
  RootBox {widgetClassPath=&HorizontalBox, widget=RootBox}
  VerticalBox_341 {
   widgetClassPath=&VerticalBox
   parent=RootBox
   widget=VerticalBox_341
   slot {class=HorizontalBoxSlot, properties {size {sizeRule=Fill, value=1}}}
  }
  Spacer_125 {widgetClassPath=&Spacer, parent=VerticalBox_341, widget=Spacer_125, properties {size=<0 12>}, slot {class=VerticalBoxSlot}}
  HorizontalBox_45 {widgetClassPath=&HorizontalBox, parent=VerticalBox_341, widget=HorizontalBox_45, slot {class=VerticalBoxSlot}}
  LabelText {
   widgetClassPath=&TextBlock
   bIsVariable=true
   parent=HorizontalBox_45
   widget=LabelText
   properties {
    text=Text Block
    font {
     bForceMonospaced=false
     bMaterialIsStencil=false
     fontMaterial=null
     fontObject=&Roboto
     letterSpacing=0
     monospacedWidth=1
     size=19.5
     skewAmount=0
     typefaceFontName=Bold
     outlineSettings {
      bApplyOutlineToDropShadows=false
      bMiteredCorners=false
      bSeparateFillAlpha=false
      outlineMaterial=null
      outlineSize=0
      outlineColor=<rgba 0 0 0 1>
     }
    }
   }
   slot {
    class=HorizontalBoxSlot
    properties {horizontalAlignment=HAlign_Center, verticalAlignment=VAlign_Center, padding=<ltrb 4 2 4 2>, size {sizeRule=Fill, value=1}}
   }
  }
  Spacer_2 {widgetClassPath=&Spacer, parent=HorizontalBox_45, widget=Spacer_2, properties {size=<50 1>}, slot {class=HorizontalBoxSlot}}
  VInput {
   widgetClassPath=&EditableTextBox
   bIsVariable=true
   parent=HorizontalBox_45
   widget=VInput
   properties {
    hintText=1.0
    toolTipText=Volume multiplier.
    widgetStyle {
     backgroundColor {colorUseRule=UseColor_Specified, specifiedColor=<rgba 1 1 1 1>}
     backgroundImageFocused {
      bIsDynamicallyLoaded=false
      drawAs=RoundedBox
      imageType=FullColor
      mirroring=NoMirror
      resourceName=../../../Engine/Content/Slate/Common/TextBox.png
      resourceObject=null
      tiling=NoTile
      imageSize=<0 0>
      margin=<ltrb 0.25 0.25 0.25 0.25>
      outlineSettings {
       bUseBrushTransparency=false
       roundingType=FixedRadius
       width=0
       color {colorUseRule=UseColor_Specified, specifiedColor=<rgba 0 0 0 0>}
       cornerRadii=<12 12 12 12>
      }
      tintColor {colorUseRule=UseColor_Specified, specifiedColor=<rgba 0 0 0 1>}
      uVRegion {bIsValid=false, max=<0 0>, min=<0 0>}
     }
     backgroundImageHovered {
      bIsDynamicallyLoaded=false
      drawAs=RoundedBox
      imageType=FullColor
      mirroring=NoMirror
      resourceName=../../../Engine/Content/Slate/Common/TextBox.png
      resourceObject=null
      tiling=NoTile
      imageSize=<0 0>
      margin=<ltrb 0.25 0.25 0.25 0.25>
      outlineSettings {
       bUseBrushTransparency=false
       roundingType=FixedRadius
       width=0
       color {colorUseRule=UseColor_Specified, specifiedColor=<rgba 0 0 0 0>}
       cornerRadii=<12 12 12 12>
      }
      tintColor {colorUseRule=UseColor_Specified, specifiedColor=<rgba 0 0 0 1>}
      uVRegion {bIsValid=false, max=<0 0>, min=<0 0>}
     }
     backgroundImageNormal {
      bIsDynamicallyLoaded=false
      drawAs=RoundedBox
      imageType=FullColor
      mirroring=NoMirror
      resourceName=../../../Engine/Content/Slate/Common/TextBox.png
      resourceObject=null
      tiling=NoTile
      imageSize=<0 0>
      margin=<ltrb 0.25 0.25 0.25 0.25>
      outlineSettings {
       bUseBrushTransparency=false
       roundingType=FixedRadius
       width=0
       color {colorUseRule=UseColor_Specified, specifiedColor=<rgba 0 0 0 0>}
       cornerRadii=<12 12 12 12>
      }
      tintColor {colorUseRule=UseColor_Specified, specifiedColor=<rgba 0 0 0 1>}
      uVRegion {bIsValid=false, max=<0 0>, min=<0 0>}
     }
     backgroundImageReadOnly {
      bIsDynamicallyLoaded=false
      drawAs=Box
      imageType=FullColor
      mirroring=NoMirror
      resourceName=../../../Engine/Content/Slate/Common/TextBox_ReadOnly.png
      resourceObject=null
      tiling=NoTile
      imageSize=<0 0>
      margin=<ltrb 0.25 0.25 0.25 0.25>
      outlineSettings {
       bUseBrushTransparency=false
       roundingType=HalfHeightRadius
       width=0
       color {colorUseRule=UseColor_Specified, specifiedColor=<rgba 0 0 0 0>}
       cornerRadii=<0 0 0 0>
      }
      tintColor {colorUseRule=UseColor_Specified, specifiedColor=<rgba 1 1 1 1>}
      uVRegion {bIsValid=false, max=<0 0>, min=<0 0>}
     }
     focusedForegroundColor {colorUseRule=UseColor_Foreground, specifiedColor=<rgba 1 0 1 1>}
     foregroundColor {colorUseRule=UseColor_Foreground, specifiedColor=<rgba 1 0 1 1>}
     hScrollBarPadding=<ltrb 0 0 0 0>
     padding=<ltrb 4 2 4 2>
     readOnlyForegroundColor {colorUseRule=UseColor_Foreground, specifiedColor=<rgba 1 0 1 1>}
     scrollBarStyle {
      thickness=16
      draggedThumbImage {
       bIsDynamicallyLoaded=false
       drawAs=Box
       imageType=FullColor
       mirroring=NoMirror
       resourceName=../../../Engine/Content/Slate/Common/Scrollbar_Thumb.png
       resourceObject=null
       tiling=NoTile
       imageSize=<0 0>
       margin=<ltrb 0.25 0.25 0.25 0.25>
       outlineSettings {
        bUseBrushTransparency=false
        roundingType=HalfHeightRadius
        width=0
        color {colorUseRule=UseColor_Specified, specifiedColor=<rgba 0 0 0 0>}
        cornerRadii=<0 0 0 0>
       }
       tintColor {colorUseRule=UseColor_Specified, specifiedColor=<rgba 1 1 1 1>}
       uVRegion {bIsValid=false, max=<0 0>, min=<0 0>}
      }
      horizontalBackgroundImage {
       bIsDynamicallyLoaded=false
       drawAs=NoDrawType
       imageType=NoImage
       mirroring=NoMirror
       resourceName=None
       resourceObject=null
       tiling=NoTile
       imageSize=<0 0>
       margin=<ltrb 0 0 0 0>
       outlineSettings {
        bUseBrushTransparency=false
        roundingType=HalfHeightRadius
        width=0
        color {colorUseRule=UseColor_Specified, specifiedColor=<rgba 0 0 0 0>}
        cornerRadii=<0 0 0 0>
       }
       tintColor {colorUseRule=UseColor_Specified, specifiedColor=<rgba 1 1 1 1>}
       uVRegion {bIsValid=false, max=<0 0>, min=<0 0>}
      }
      horizontalBottomSlotImage {
       bIsDynamicallyLoaded=false
       drawAs=Image
       imageType=FullColor
       mirroring=NoMirror
       resourceName=../../../Engine/Content/Slate/Common/Scrollbar_Background_Horizontal.png
       resourceObject=null
       tiling=NoTile
       imageSize=<8 8>
       margin=<ltrb 0 0 0 0>
       outlineSettings {
        bUseBrushTransparency=false
        roundingType=HalfHeightRadius
        width=0
        color {colorUseRule=UseColor_Specified, specifiedColor=<rgba 0 0 0 0>}
        cornerRadii=<0 0 0 0>
       }
       tintColor {colorUseRule=UseColor_Specified, specifiedColor=<rgba 1 1 1 1>}
       uVRegion {bIsValid=false, max=<0 0>, min=<0 0>}
      }
      horizontalTopSlotImage {
       bIsDynamicallyLoaded=false
       drawAs=Image
       imageType=FullColor
       mirroring=NoMirror
       resourceName=../../../Engine/Content/Slate/Common/Scrollbar_Background_Horizontal.png
       resourceObject=null
       tiling=NoTile
       imageSize=<8 8>
       margin=<ltrb 0 0 0 0>
       outlineSettings {
        bUseBrushTransparency=false
        roundingType=HalfHeightRadius
        width=0
        color {colorUseRule=UseColor_Specified, specifiedColor=<rgba 0 0 0 0>}
        cornerRadii=<0 0 0 0>
       }
       tintColor {colorUseRule=UseColor_Specified, specifiedColor=<rgba 1 1 1 1>}
       uVRegion {bIsValid=false, max=<0 0>, min=<0 0>}
      }
      hoveredThumbImage {
       bIsDynamicallyLoaded=false
       drawAs=Box
       imageType=FullColor
       mirroring=NoMirror
       resourceName=../../../Engine/Content/Slate/Common/Scrollbar_Thumb.png
       resourceObject=null
       tiling=NoTile
       imageSize=<0 0>
       margin=<ltrb 0.25 0.25 0.25 0.25>
       outlineSettings {
        bUseBrushTransparency=false
        roundingType=HalfHeightRadius
        width=0
        color {colorUseRule=UseColor_Specified, specifiedColor=<rgba 0 0 0 0>}
        cornerRadii=<0 0 0 0>
       }
       tintColor {colorUseRule=UseColor_Specified, specifiedColor=<rgba 1 1 1 1>}
       uVRegion {bIsValid=false, max=<0 0>, min=<0 0>}
      }
      normalThumbImage {
       bIsDynamicallyLoaded=false
       drawAs=Box
       imageType=FullColor
       mirroring=NoMirror
       resourceName=../../../Engine/Content/Slate/Common/Scrollbar_Thumb.png
       resourceObject=null
       tiling=NoTile
       imageSize=<0 0>
       margin=<ltrb 0.25 0.25 0.25 0.25>
       outlineSettings {
        bUseBrushTransparency=false
        roundingType=HalfHeightRadius
        width=0
        color {colorUseRule=UseColor_Specified, specifiedColor=<rgba 0 0 0 0>}
        cornerRadii=<0 0 0 0>
       }
       tintColor {colorUseRule=UseColor_Specified, specifiedColor=<rgba 1 1 1 1>}
       uVRegion {bIsValid=false, max=<0 0>, min=<0 0>}
      }
      verticalBackgroundImage {
       bIsDynamicallyLoaded=false
       drawAs=NoDrawType
       imageType=NoImage
       mirroring=NoMirror
       resourceName=None
       resourceObject=null
       tiling=NoTile
       imageSize=<0 0>
       margin=<ltrb 0 0 0 0>
       outlineSettings {
        bUseBrushTransparency=false
        roundingType=HalfHeightRadius
        width=0
        color {colorUseRule=UseColor_Specified, specifiedColor=<rgba 0 0 0 0>}
        cornerRadii=<0 0 0 0>
       }
       tintColor {colorUseRule=UseColor_Specified, specifiedColor=<rgba 1 1 1 1>}
       uVRegion {bIsValid=false, max=<0 0>, min=<0 0>}
      }
      verticalBottomSlotImage {
       bIsDynamicallyLoaded=false
       drawAs=Image
       imageType=FullColor
       mirroring=NoMirror
       resourceName=../../../Engine/Content/Slate/Common/Scrollbar_Background_Vertical.png
       resourceObject=null
       tiling=NoTile
       imageSize=<8 8>
       margin=<ltrb 0 0 0 0>
       outlineSettings {
        bUseBrushTransparency=false
        roundingType=HalfHeightRadius
        width=0
        color {colorUseRule=UseColor_Specified, specifiedColor=<rgba 0 0 0 0>}
        cornerRadii=<0 0 0 0>
       }
       tintColor {colorUseRule=UseColor_Specified, specifiedColor=<rgba 1 1 1 1>}
       uVRegion {bIsValid=false, max=<0 0>, min=<0 0>}
      }
      verticalTopSlotImage {
       bIsDynamicallyLoaded=false
       drawAs=Image
       imageType=FullColor
       mirroring=NoMirror
       resourceName=../../../Engine/Content/Slate/Common/Scrollbar_Background_Vertical.png
       resourceObject=null
       tiling=NoTile
       imageSize=<8 8>
       margin=<ltrb 0 0 0 0>
       outlineSettings {
        bUseBrushTransparency=false
        roundingType=HalfHeightRadius
        width=0
        color {colorUseRule=UseColor_Specified, specifiedColor=<rgba 0 0 0 0>}
        cornerRadii=<0 0 0 0>
       }
       tintColor {colorUseRule=UseColor_Specified, specifiedColor=<rgba 1 1 1 1>}
       uVRegion {bIsValid=false, max=<0 0>, min=<0 0>}
      }
     }
     textStyle {
      overflowPolicy=Clip
      transformPolicy=None
      colorAndOpacity {colorUseRule=UseColor_Foreground, specifiedColor=<rgba 1 0 1 1>}
      font {
       bForceMonospaced=false
       bMaterialIsStencil=false
       fontMaterial=null
       fontObject=&Roboto
       letterSpacing=0
       monospacedWidth=1
       size=15
       skewAmount=0
       typefaceFontName=Regular
       outlineSettings {
        bApplyOutlineToDropShadows=false
        bMiteredCorners=false
        bSeparateFillAlpha=false
        outlineMaterial=null
        outlineSize=0
        outlineColor=<rgba 0 0 0 1>
       }
      }
      highlightColor {colorUseRule=UseColor_Specified, specifiedColor=<rgba 0.02 0.3 0 1>}
      highlightShape {
       bIsDynamicallyLoaded=false
       drawAs=Box
       imageType=FullColor
       mirroring=NoMirror
       resourceName=../../../Engine/Content/Slate/Common/TextBlockHighlightShape.png
       resourceObject=null
       tiling=NoTile
       imageSize=<0 0>
       margin=<ltrb 0.375 0.375 0.375 0.375>
       outlineSettings {
        bUseBrushTransparency=false
        roundingType=HalfHeightRadius
        width=0
        color {colorUseRule=UseColor_Specified, specifiedColor=<rgba 0 0 0 0>}
        cornerRadii=<0 0 0 0>
       }
       tintColor {colorUseRule=UseColor_Specified, specifiedColor=<rgba 1 1 1 1>}
       uVRegion {bIsValid=false, max=<0 0>, min=<0 0>}
      }
      selectedBackgroundColor {colorUseRule=UseColor_Foreground, specifiedColor=<rgba 1 0 1 1>}
      shadowColorAndOpacity=<rgba 0 0 0 1>
      shadowOffset=<0 0>
      strikeBrush {
       bIsDynamicallyLoaded=false
       drawAs=Image
       imageType=NoImage
       mirroring=NoMirror
       resourceName=None
       resourceObject=null
       tiling=NoTile
       imageSize=<32 32>
       margin=<ltrb 0 0 0 0>
       outlineSettings {
        bUseBrushTransparency=false
        roundingType=HalfHeightRadius
        width=0
        color {colorUseRule=UseColor_Specified, specifiedColor=<rgba 0 0 0 0>}
        cornerRadii=<0 0 0 0>
       }
       tintColor {colorUseRule=UseColor_Specified, specifiedColor=<rgba 1 1 1 1>}
       uVRegion {bIsValid=false, max=<0 0>, min=<0 0>}
      }
      underlineBrush {
       bIsDynamicallyLoaded=false
       drawAs=Image
       imageType=NoImage
       mirroring=NoMirror
       resourceName=None
       resourceObject=null
       tiling=NoTile
       imageSize=<32 32>
       margin=<ltrb 0 0 0 0>
       outlineSettings {
        bUseBrushTransparency=false
        roundingType=HalfHeightRadius
        width=0
        color {colorUseRule=UseColor_Specified, specifiedColor=<rgba 0 0 0 0>}
        cornerRadii=<0 0 0 0>
       }
       tintColor {colorUseRule=UseColor_Specified, specifiedColor=<rgba 1 1 1 1>}
       uVRegion {bIsValid=false, max=<0 0>, min=<0 0>}
      }
     }
     vScrollBarPadding=<ltrb 0 0 0 0>
    }
   }
   slot {class=HorizontalBoxSlot, properties {verticalAlignment=VAlign_Center, padding=<ltrb 4 2 4 2>, size {sizeRule=Fill, value=0.1}}}
  }
  PInput {
   widgetClassPath=&EditableTextBox
   bIsVariable=true
   parent=HorizontalBox_45
   widget=PInput
   properties {
    hintText=1.0
    toolTipText=Pitch multiplier
    widgetStyle {
     backgroundColor {colorUseRule=UseColor_Specified, specifiedColor=<rgba 1 1 1 1>}
     backgroundImageFocused {
      bIsDynamicallyLoaded=false
      drawAs=RoundedBox
      imageType=FullColor
      mirroring=NoMirror
      resourceName=../../../Engine/Content/Slate/Common/TextBox.png
      resourceObject=null
      tiling=NoTile
      imageSize=<0 0>
      margin=<ltrb 0.25 0.25 0.25 0.25>
      outlineSettings {
       bUseBrushTransparency=false
       roundingType=FixedRadius
       width=1
       color {colorUseRule=UseColor_Specified, specifiedColor=<rgba 0 0 0 0>}
       cornerRadii=<12 12 12 12>
      }
      tintColor {colorUseRule=UseColor_Specified, specifiedColor=<rgba 0 0 0 1>}
      uVRegion {bIsValid=false, max=<0 0>, min=<0 0>}
     }
     backgroundImageHovered {
      bIsDynamicallyLoaded=false
      drawAs=RoundedBox
      imageType=FullColor
      mirroring=NoMirror
      resourceName=../../../Engine/Content/Slate/Common/TextBox.png
      resourceObject=null
      tiling=NoTile
      imageSize=<0 0>
      margin=<ltrb 0.25 0.25 0.25 0.25>
      outlineSettings {
       bUseBrushTransparency=false
       roundingType=FixedRadius
       width=1
       color {colorUseRule=UseColor_Specified, specifiedColor=<rgba 0 0 0 0>}
       cornerRadii=<12 12 12 12>
      }
      tintColor {colorUseRule=UseColor_Specified, specifiedColor=<rgba 0 0 0 1>}
      uVRegion {bIsValid=false, max=<0 0>, min=<0 0>}
     }
     backgroundImageNormal {
      bIsDynamicallyLoaded=false
      drawAs=RoundedBox
      imageType=FullColor
      mirroring=NoMirror
      resourceName=../../../Engine/Content/Slate/Common/TextBox.png
      resourceObject=null
      tiling=NoTile
      imageSize=<0 0>
      margin=<ltrb 0.25 0.25 0.25 0.25>
      outlineSettings {
       bUseBrushTransparency=false
       roundingType=FixedRadius
       width=1
       color {colorUseRule=UseColor_Specified, specifiedColor=<rgba 0 0 0 0>}
       cornerRadii=<12 12 12 12>
      }
      tintColor {colorUseRule=UseColor_Specified, specifiedColor=<rgba 0 0 0 1>}
      uVRegion {bIsValid=false, max=<0 0>, min=<0 0>}
     }
     backgroundImageReadOnly {
      bIsDynamicallyLoaded=false
      drawAs=Box
      imageType=FullColor
      mirroring=NoMirror
      resourceName=../../../Engine/Content/Slate/Common/TextBox_ReadOnly.png
      resourceObject=null
      tiling=NoTile
      imageSize=<0 0>
      margin=<ltrb 0.25 0.25 0.25 0.25>
      outlineSettings {
       bUseBrushTransparency=false
       roundingType=HalfHeightRadius
       width=0
       color {colorUseRule=UseColor_Specified, specifiedColor=<rgba 0 0 0 0>}
       cornerRadii=<0 0 0 0>
      }
      tintColor {colorUseRule=UseColor_Specified, specifiedColor=<rgba 1 1 1 1>}
      uVRegion {bIsValid=false, max=<0 0>, min=<0 0>}
     }
     focusedForegroundColor {colorUseRule=UseColor_Foreground, specifiedColor=<rgba 1 0 1 1>}
     foregroundColor {colorUseRule=UseColor_Foreground, specifiedColor=<rgba 1 0 1 1>}
     hScrollBarPadding=<ltrb 0 0 0 0>
     padding=<ltrb 4 2 4 2>
     readOnlyForegroundColor {colorUseRule=UseColor_Foreground, specifiedColor=<rgba 1 0 1 1>}
     scrollBarStyle {
      thickness=16
      draggedThumbImage {
       bIsDynamicallyLoaded=false
       drawAs=Box
       imageType=FullColor
       mirroring=NoMirror
       resourceName=../../../Engine/Content/Slate/Common/Scrollbar_Thumb.png
       resourceObject=null
       tiling=NoTile
       imageSize=<0 0>
       margin=<ltrb 0.25 0.25 0.25 0.25>
       outlineSettings {
        bUseBrushTransparency=false
        roundingType=HalfHeightRadius
        width=0
        color {colorUseRule=UseColor_Specified, specifiedColor=<rgba 0 0 0 0>}
        cornerRadii=<0 0 0 0>
       }
       tintColor {colorUseRule=UseColor_Specified, specifiedColor=<rgba 1 1 1 1>}
       uVRegion {bIsValid=false, max=<0 0>, min=<0 0>}
      }
      horizontalBackgroundImage {
       bIsDynamicallyLoaded=false
       drawAs=NoDrawType
       imageType=NoImage
       mirroring=NoMirror
       resourceName=None
       resourceObject=null
       tiling=NoTile
       imageSize=<0 0>
       margin=<ltrb 0 0 0 0>
       outlineSettings {
        bUseBrushTransparency=false
        roundingType=HalfHeightRadius
        width=0
        color {colorUseRule=UseColor_Specified, specifiedColor=<rgba 0 0 0 0>}
        cornerRadii=<0 0 0 0>
       }
       tintColor {colorUseRule=UseColor_Specified, specifiedColor=<rgba 1 1 1 1>}
       uVRegion {bIsValid=false, max=<0 0>, min=<0 0>}
      }
      horizontalBottomSlotImage {
       bIsDynamicallyLoaded=false
       drawAs=Image
       imageType=FullColor
       mirroring=NoMirror
       resourceName=../../../Engine/Content/Slate/Common/Scrollbar_Background_Horizontal.png
       resourceObject=null
       tiling=NoTile
       imageSize=<8 8>
       margin=<ltrb 0 0 0 0>
       outlineSettings {
        bUseBrushTransparency=false
        roundingType=HalfHeightRadius
        width=0
        color {colorUseRule=UseColor_Specified, specifiedColor=<rgba 0 0 0 0>}
        cornerRadii=<0 0 0 0>
       }
       tintColor {colorUseRule=UseColor_Specified, specifiedColor=<rgba 1 1 1 1>}
       uVRegion {bIsValid=false, max=<0 0>, min=<0 0>}
      }
      horizontalTopSlotImage {
       bIsDynamicallyLoaded=false
       drawAs=Image
       imageType=FullColor
       mirroring=NoMirror
       resourceName=../../../Engine/Content/Slate/Common/Scrollbar_Background_Horizontal.png
       resourceObject=null
       tiling=NoTile
       imageSize=<8 8>
       margin=<ltrb 0 0 0 0>
       outlineSettings {
        bUseBrushTransparency=false
        roundingType=HalfHeightRadius
        width=0
        color {colorUseRule=UseColor_Specified, specifiedColor=<rgba 0 0 0 0>}
        cornerRadii=<0 0 0 0>
       }
       tintColor {colorUseRule=UseColor_Specified, specifiedColor=<rgba 1 1 1 1>}
       uVRegion {bIsValid=false, max=<0 0>, min=<0 0>}
      }
      hoveredThumbImage {
       bIsDynamicallyLoaded=false
       drawAs=Box
       imageType=FullColor
       mirroring=NoMirror
       resourceName=../../../Engine/Content/Slate/Common/Scrollbar_Thumb.png
       resourceObject=null
       tiling=NoTile
       imageSize=<0 0>
       margin=<ltrb 0.25 0.25 0.25 0.25>
       outlineSettings {
        bUseBrushTransparency=false
        roundingType=HalfHeightRadius
        width=0
        color {colorUseRule=UseColor_Specified, specifiedColor=<rgba 0 0 0 0>}
        cornerRadii=<0 0 0 0>
       }
       tintColor {colorUseRule=UseColor_Specified, specifiedColor=<rgba 1 1 1 1>}
       uVRegion {bIsValid=false, max=<0 0>, min=<0 0>}
      }
      normalThumbImage {
       bIsDynamicallyLoaded=false
       drawAs=Box
       imageType=FullColor
       mirroring=NoMirror
       resourceName=../../../Engine/Content/Slate/Common/Scrollbar_Thumb.png
       resourceObject=null
       tiling=NoTile
       imageSize=<0 0>
       margin=<ltrb 0.25 0.25 0.25 0.25>
       outlineSettings {
        bUseBrushTransparency=false
        roundingType=HalfHeightRadius
        width=0
        color {colorUseRule=UseColor_Specified, specifiedColor=<rgba 0 0 0 0>}
        cornerRadii=<0 0 0 0>
       }
       tintColor {colorUseRule=UseColor_Specified, specifiedColor=<rgba 1 1 1 1>}
       uVRegion {bIsValid=false, max=<0 0>, min=<0 0>}
      }
      verticalBackgroundImage {
       bIsDynamicallyLoaded=false
       drawAs=NoDrawType
       imageType=NoImage
       mirroring=NoMirror
       resourceName=None
       resourceObject=null
       tiling=NoTile
       imageSize=<0 0>
       margin=<ltrb 0 0 0 0>
       outlineSettings {
        bUseBrushTransparency=false
        roundingType=HalfHeightRadius
        width=0
        color {colorUseRule=UseColor_Specified, specifiedColor=<rgba 0 0 0 0>}
        cornerRadii=<0 0 0 0>
       }
       tintColor {colorUseRule=UseColor_Specified, specifiedColor=<rgba 1 1 1 1>}
       uVRegion {bIsValid=false, max=<0 0>, min=<0 0>}
      }
      verticalBottomSlotImage {
       bIsDynamicallyLoaded=false
       drawAs=Image
       imageType=FullColor
       mirroring=NoMirror
       resourceName=../../../Engine/Content/Slate/Common/Scrollbar_Background_Vertical.png
       resourceObject=null
       tiling=NoTile
       imageSize=<8 8>
       margin=<ltrb 0 0 0 0>
       outlineSettings {
        bUseBrushTransparency=false
        roundingType=HalfHeightRadius
        width=0
        color {colorUseRule=UseColor_Specified, specifiedColor=<rgba 0 0 0 0>}
        cornerRadii=<0 0 0 0>
       }
       tintColor {colorUseRule=UseColor_Specified, specifiedColor=<rgba 1 1 1 1>}
       uVRegion {bIsValid=false, max=<0 0>, min=<0 0>}
      }
      verticalTopSlotImage {
       bIsDynamicallyLoaded=false
       drawAs=Image
       imageType=FullColor
       mirroring=NoMirror
       resourceName=../../../Engine/Content/Slate/Common/Scrollbar_Background_Vertical.png
       resourceObject=null
       tiling=NoTile
       imageSize=<8 8>
       margin=<ltrb 0 0 0 0>
       outlineSettings {
        bUseBrushTransparency=false
        roundingType=HalfHeightRadius
        width=0
        color {colorUseRule=UseColor_Specified, specifiedColor=<rgba 0 0 0 0>}
        cornerRadii=<0 0 0 0>
       }
       tintColor {colorUseRule=UseColor_Specified, specifiedColor=<rgba 1 1 1 1>}
       uVRegion {bIsValid=false, max=<0 0>, min=<0 0>}
      }
     }
     textStyle {
      overflowPolicy=Clip
      transformPolicy=None
      colorAndOpacity {colorUseRule=UseColor_Foreground, specifiedColor=<rgba 1 0 1 1>}
      font {
       bForceMonospaced=false
       bMaterialIsStencil=false
       fontMaterial=null
       fontObject=&Roboto
       letterSpacing=0
       monospacedWidth=1
       size=15
       skewAmount=0
       typefaceFontName=Regular
       outlineSettings {
        bApplyOutlineToDropShadows=false
        bMiteredCorners=false
        bSeparateFillAlpha=false
        outlineMaterial=null
        outlineSize=0
        outlineColor=<rgba 0 0 0 1>
       }
      }
      highlightColor {colorUseRule=UseColor_Specified, specifiedColor=<rgba 0.02 0.3 0 1>}
      highlightShape {
       bIsDynamicallyLoaded=false
       drawAs=Box
       imageType=FullColor
       mirroring=NoMirror
       resourceName=../../../Engine/Content/Slate/Common/TextBlockHighlightShape.png
       resourceObject=null
       tiling=NoTile
       imageSize=<0 0>
       margin=<ltrb 0.375 0.375 0.375 0.375>
       outlineSettings {
        bUseBrushTransparency=false
        roundingType=HalfHeightRadius
        width=0
        color {colorUseRule=UseColor_Specified, specifiedColor=<rgba 0 0 0 0>}
        cornerRadii=<0 0 0 0>
       }
       tintColor {colorUseRule=UseColor_Specified, specifiedColor=<rgba 1 1 1 1>}
       uVRegion {bIsValid=false, max=<0 0>, min=<0 0>}
      }
      selectedBackgroundColor {colorUseRule=UseColor_Foreground, specifiedColor=<rgba 1 0 1 1>}
      shadowColorAndOpacity=<rgba 0 0 0 1>
      shadowOffset=<0 0>
      strikeBrush {
       bIsDynamicallyLoaded=false
       drawAs=Image
       imageType=NoImage
       mirroring=NoMirror
       resourceName=None
       resourceObject=null
       tiling=NoTile
       imageSize=<32 32>
       margin=<ltrb 0 0 0 0>
       outlineSettings {
        bUseBrushTransparency=false
        roundingType=HalfHeightRadius
        width=0
        color {colorUseRule=UseColor_Specified, specifiedColor=<rgba 0 0 0 0>}
        cornerRadii=<0 0 0 0>
       }
       tintColor {colorUseRule=UseColor_Specified, specifiedColor=<rgba 1 1 1 1>}
       uVRegion {bIsValid=false, max=<0 0>, min=<0 0>}
      }
      underlineBrush {
       bIsDynamicallyLoaded=false
       drawAs=Image
       imageType=NoImage
       mirroring=NoMirror
       resourceName=None
       resourceObject=null
       tiling=NoTile
       imageSize=<32 32>
       margin=<ltrb 0 0 0 0>
       outlineSettings {
        bUseBrushTransparency=false
        roundingType=HalfHeightRadius
        width=0
        color {colorUseRule=UseColor_Specified, specifiedColor=<rgba 0 0 0 0>}
        cornerRadii=<0 0 0 0>
       }
       tintColor {colorUseRule=UseColor_Specified, specifiedColor=<rgba 1 1 1 1>}
       uVRegion {bIsValid=false, max=<0 0>, min=<0 0>}
      }
     }
     vScrollBarPadding=<ltrb 0 0 0 0>
    }
   }
   slot {class=HorizontalBoxSlot, properties {verticalAlignment=VAlign_Center, padding=<ltrb 4 2 4 2>, size {sizeRule=Fill, value=0.1}}}
  }
  Spacer_182 {widgetClassPath=&Spacer, parent=HorizontalBox_45, widget=Spacer_182, properties {size=<50 1>}, slot {class=HorizontalBoxSlot}}
  CopyButton {
   widgetClassPath=&Button
   parent=HorizontalBox_45
   widget=CopyButton
   properties {widgetStyle {
     clickedSlateSound {resourceObject=null}
     disabled {
      bIsDynamicallyLoaded=false
      drawAs=NoDrawType
      imageType=NoImage
      mirroring=NoMirror
      resourceName=None
      resourceObject=null
      tiling=NoTile
      imageSize=<0 0>
      margin=<ltrb 0 0 0 0>
      outlineSettings {
       bUseBrushTransparency=false
       roundingType=HalfHeightRadius
       width=0
       color {colorUseRule=UseColor_Specified, specifiedColor=<rgba 0 0 0 0>}
       cornerRadii=<0 0 0 0>
      }
      tintColor {colorUseRule=UseColor_Specified, specifiedColor=<rgba 1 1 1 1>}
      uVRegion {bIsValid=false, max=<0 0>, min=<0 0>}
     }
     disabledForeground {colorUseRule=UseColor_Specified, specifiedColor=<rgba 0.5271 0.5271 0.5271 1>}
     hovered {
      bIsDynamicallyLoaded=false
      drawAs=RoundedBox
      imageType=NoImage
      mirroring=NoMirror
      resourceName=None
      resourceObject=null
      tiling=NoTile
      imageSize=<32 32>
      margin=<ltrb 0 0 0 0>
      outlineSettings {
       bUseBrushTransparency=true
       roundingType=FixedRadius
       width=1
       color {colorUseRule=UseColor_Specified, specifiedColor=<rgba 0.0756 0.3559 0.901 1>}
       cornerRadii=<12 12 12 12>
      }
      tintColor {colorUseRule=UseColor_Specified, specifiedColor=<rgba 0.0756 0.3559 0.901 1>}
      uVRegion {bIsValid=false, max=<0 0>, min=<0 0>}
     }
     hoveredForeground {colorUseRule=UseColor_Specified, specifiedColor=<rgba 0 1 1 1>}
     hoveredSlateSound {resourceObject=null}
     normal {
      bIsDynamicallyLoaded=false
      drawAs=RoundedBox
      imageType=NoImage
      mirroring=NoMirror
      resourceName=None
      resourceObject=null
      tiling=NoTile
      imageSize=<32 32>
      margin=<ltrb 0 0 0 0>
      outlineSettings {
       bUseBrushTransparency=true
       roundingType=FixedRadius
       width=1
       color {colorUseRule=UseColor_Specified, specifiedColor=<rgba 0.0643 0.3024 0.7656 1>}
       cornerRadii=<12 12 12 12>
      }
      tintColor {colorUseRule=UseColor_Specified, specifiedColor=<rgba 0.0643 0.3024 0.7656 1>}
      uVRegion {bIsValid=false, max=<0 0>, min=<0 0>}
     }
     normalForeground {colorUseRule=UseColor_Specified, specifiedColor=<rgba 0 1 1 1>}
     normalPadding=<ltrb 12 1.5 12 1.5>
     pressed {
      bIsDynamicallyLoaded=false
      drawAs=RoundedBox
      imageType=NoImage
      mirroring=NoMirror
      resourceName=None
      resourceObject=null
      tiling=NoTile
      imageSize=<32 32>
      margin=<ltrb 0 0 0 0>
      outlineSettings {
       bUseBrushTransparency=true
       roundingType=FixedRadius
       width=1
       color {colorUseRule=UseColor_Specified, specifiedColor=<rgba 0.5463 0.6208 0.7656 1>}
       cornerRadii=<12 12 12 12>
      }
      tintColor {colorUseRule=UseColor_Specified, specifiedColor=<rgba 0.5463 0.6208 0.7656 1>}
      uVRegion {bIsValid=false, max=<0 0>, min=<0 0>}
     }
     pressedForeground {colorUseRule=UseColor_Specified, specifiedColor=<rgba 0 1 1 1>}
     pressedPadding=<ltrb 12 1.5 12 1.5>
     pressedSlateSound {resourceObject=null}
    }}
   slot {class=HorizontalBoxSlot, properties {verticalAlignment=VAlign_Center, padding=<ltrb 4 2 4 2>, size {sizeRule=Fill, value=0.3}}}
  }
  CopyButtonLabel {
   widgetClassPath=&TextBlock
   parent=CopyButton
   widget=CopyButtonLabel
   properties {
    text=Copy
    font {
     bForceMonospaced=false
     bMaterialIsStencil=false
     fontMaterial=null
     fontObject=&Roboto
     letterSpacing=0
     monospacedWidth=1
     size=19.5
     skewAmount=0
     typefaceFontName=Bold
     outlineSettings {
      bApplyOutlineToDropShadows=false
      bMiteredCorners=false
      bSeparateFillAlpha=false
      outlineMaterial=null
      outlineSize=0
      outlineColor=<rgba 0 0 0 1>
     }
    }
   }
   slot {class=ButtonSlot}
  }
  Spacer {widgetClassPath=&Spacer, parent=VerticalBox_341, widget=Spacer, properties {size=<0 12>}, slot {class=VerticalBoxSlot}}
 ]
}
