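# Editor Utility Widget "Foley Clipboard": a palette of foley stamps such as footstep, scuff, jump and land for audio authoring.
# Construct builds nine hardcoded WBP_FoleyClipboardRow widgets, each with kind, tag and number hotkey.
# HandleStampClicked passes key, volume and pitch to FoleyClipboard.load_clipboard through concatenated Python commands; Reset restores rows, then calls reset_palette.
# Mouse enter takes keyboard focus so OnKeyDown can offer keys to each row's TryHandleHotkey.
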EditorUtilityWidgetBlueprint /FortTechAudioTools/Python/FoleyClipboard/EUW_FoleyClipboard
graphs [
 BuildRow {nodes [
   n1 = BuildRow!() -> (then, InKey:Name, InLabel:Text, InKind:Name, InTag:Name, InHotkey:Name) {is=title K2Node_FunctionEntry}
   > n2 = Create Widget!(
    execute:Exec
    WorldContextObject:Object Reference=Self-Reference()
    Class:User Widget Class Reference=&/WBP_FoleyClipboardRow.WBP_FoleyClipboardRow_C
    OwningPlayer:Player Controller Object Reference
   ) -> (then:Exec, ReturnValue:WBP Foley Clipboard Row Object Reference)
   > InitFromConfig!(
    execute
    self:WBP Foley Clipboard Row Object Reference=@n2
    InKey:Name=@n1
    InLabel:Text=@n1.InLabel
    InKind:Name=@n1.InKind
    InTag:Name=@n1.InTag
    InHotkey:Name=@n1.InHotkey
   ) -> (then)
   > AddChildToVerticalBox(self=Get RowsContainer() -> (RowsContainer:Vertical Box Object Reference), Content=@n2)
   > Add(
    TargetArray:Array of WBP Foley Clipboard Row Object References=Get GeneratedRowWidgets() -> (GeneratedRowWidgets:Array of WBP Foley Clipboard Row Object References)
    NewItem:"WBP Foley Clipboard Row Object Reference (by ref)"=@n2
   ) {is=title K2Node_CallArrayFunction}
   > Bind Event to On Stamp Clicked(self:WBP Foley Clipboard Row Object Reference=@n2, Delegate=Create Event()) {is=title K2Node_AddDelegate}
  ]}
 EventGraph {nodes [
   "On Clicked (ResetButton)"() {is=title K2Node_ComponentBoundEvent}
   > n1 = For Each Loop(Array:Array of WBP Foley Clipboard Row Object References=Get GeneratedRowWidgets()) -> (Array Element:WBP Foley Clipboard Row Object Reference)
   ExecutePythonCommand(execute=@n1.Completed, PythonCommand="from FoleyClipboard import reset_palette; reset_palette()")
   ResetToDefaults(execute=@n1.LoopBody, self:WBP Foley Clipboard Row Object Reference=@n1)
   Event OnMouseEnter()
   > SetKeyboardFocus(self:Widget Object Reference)
   Event Construct()
   > BuildRow!(
    execute
    self
    InKey:Name=MakeLiteralName(Value=footstep)
    InLabel:Text=MakeLiteralText(Value="LOCTEXT(\"Footstep\")")
    InKind:Name=MakeLiteralName(Value=footstep)
    InTag:Name=MakeLiteralName()
    InHotkey:Name=MakeLiteralName(Value=1)
   ) -> (then) {is=title K2Node_CallFunction}
   > BuildRow(
    InKey=MakeLiteralName(Value=scuff)
    InLabel=MakeLiteralText(Value="LOCTEXT(\"Scuff\")")
    InKind=MakeLiteralName(Value=sl)
    InTag=MakeLiteralName(Value=SoundLibrary.Foley.Scuff)
    InHotkey=MakeLiteralName(Value=2)
   )
   > BuildRow(
    InKey=MakeLiteralName(Value=turn)
    InLabel=MakeLiteralText(Value="LOCTEXT(\"Turn\")")
    InKind=MakeLiteralName(Value=sl)
    InTag=MakeLiteralName(Value=SoundLibrary.Foley.Turn)
    InHotkey=MakeLiteralName(Value=3)
   )
   > BuildRow(
    InKey=MakeLiteralName(Value=land)
    InLabel=MakeLiteralText(Value="LOCTEXT(\"Land\")")
    InKind=MakeLiteralName(Value=sl)
    InTag=MakeLiteralName(Value=SoundLibrary.Foley.Land)
    InHotkey=MakeLiteralName(Value=4)
   )
   > BuildRow(
    InKey=MakeLiteralName(Value=jump)
    InLabel=MakeLiteralText(Value="LOCTEXT(\"Jump\")")
    InKind=MakeLiteralName(Value=sl)
    InTag=MakeLiteralName(Value=SoundLibrary.Foley.Jump)
    InHotkey=MakeLiteralName(Value=5)
   )
   > BuildRow(
    InKey=MakeLiteralName(Value=swish)
    InLabel=MakeLiteralText(Value="LOCTEXT(\"Swish\")")
    InKind=MakeLiteralName(Value=sl)
    InTag=MakeLiteralName(Value=SoundLibrary.Foley.SwishShort)
    InHotkey=MakeLiteralName(Value=6)
   )
   > BuildRow(
    InKey=MakeLiteralName(Value=turn_clothing)
    InLabel=MakeLiteralText(Value="LOCTEXT(\"Turn (Cloth)\")")
    InKind=MakeLiteralName(Value=sl)
    InTag=MakeLiteralName(Value=SoundLibrary.Foley.Turn.Clothing)
    InHotkey=MakeLiteralName(Value=7)
   )
   > BuildRow(
    InKey=MakeLiteralName(Value=crouch)
    InLabel=MakeLiteralText(Value="LOCTEXT(\"Crouch\")")
    InKind=MakeLiteralName(Value=sl)
    InTag=MakeLiteralName(Value=SoundLibrary.Foley.Crouch)
    InHotkey=MakeLiteralName(Value=8)
   )
   > BuildRow(
    InKey=MakeLiteralName(Value=stand)
    InLabel=MakeLiteralText(Value="LOCTEXT(\"Stand\")")
    InKind=MakeLiteralName(Value=sl)
    InTag=MakeLiteralName(Value=SoundLibrary.Foley.Stand)
    InHotkey=MakeLiteralName(Value=9)
   )
  ]}
 HandleStampClicked {nodes [
   n1 = HandleStampClicked!() -> (then, Key:Name, VolText:Text, PitchText:Text) {is=title K2Node_FunctionEntry}
   n2 = "To String (Text)"(InText=@n1.PitchText)
   n3 = Append(A="from FoleyClipboard import load_clipboard; load_clipboard('", B="To String (Name)"(InName=@n1))
   n4 = Append(A=Append(A=Append(A=@n3, B="', '"), B="To String (Text)"(InText=@n1.VolText)), B="', '")
   > ExecutePythonCommand(PythonCommand=Append(A=Append(A=@n4, B=@n2), B="')"))
  ]}
 OnKeyDown {nodes [
   n1 = OnKeyDown!() -> (then, MyGeometry:Geometry Structure, InKeyEvent:Key Event Structure) {is=title K2Node_FunctionEntry}
   > n2 = For Each Loop(Array=Get GeneratedRowWidgets())
   Return Node!(execute=@n2.Completed, ReturnValue:Event Reply Structure=Unhandled())
   n3 = GetKey!(Input:"Key Event Structure (by ref)"=@n1.InKeyEvent) -> (ReturnValue:Key Structure)
   > n4 = TryHandleHotkey!(
    execute
    self:WBP Foley Clipboard Row Object Reference=@n2
    InKeyString:String="To String (Text)"(InText=Get Key Display Name!(Key:"Key Structure (by ref)"=@n3, bLongDisplayName:bool=true) -> (ReturnValue:Text) {is=title K2Node_CallFunction})
   ) -> (then, bHandled:bool)
   > Branch(Condition=@n4)
   > Return Node(ReturnValue=Handled())
  ]}
]
variables [
 bJsonError {type=boolean}
 GeneratedRowWidgets {type="array<WBP_FoleyClipboardRow_C>", default []}
 JsonErrorMessage {type=string, default=""}
 RowHotkeys {type="array<string>", default []}
 RowKeys {type="array<string>", default []}
 RowKinds {type="array<string>", default []}
 RowLabels {type="array<string>", default []}
 RowTags {type="array<string>", default []}
]
widgets {
 info {rootWidgetClass=&VerticalBox, widgetCount=11}
 widgets [
  RootBox {widgetClassPath=&VerticalBox, widget=RootBox}
  EditorUtilityScrollBox_30 {
   widgetClassPath=&EditorUtilityScrollBox
   parent=RootBox
   widget=EditorUtilityScrollBox_30
   slot {class=VerticalBoxSlot, properties {size {sizeRule=Fill, value=1}}}
  }
  Border_59 {
   widgetClassPath=&Border
   parent=EditorUtilityScrollBox_30
   widget=Border_59
   properties {
    background {
     bIsDynamicallyLoaded=false
     drawAs=RoundedBox
     imageType=NoImage
     mirroring=NoMirror
     resourceName=None
     resourceObject=null
     tiling=NoTile
     imageSize=<32 32>
     margin=<ltrb 0 0 0 0>
     outlineSettings {
      bUseBrushTransparency=false
      roundingType=FixedRadius
      width=0
      color {colorUseRule=UseColor_Specified, specifiedColor=<rgba 0 0 0 0>}
      cornerRadii=<12 12 12 12>
     }
     tintColor {colorUseRule=UseColor_Specified, specifiedColor=<rgba 0.0156 0.0156 0.0156 1>}
     uVRegion {bIsValid=false, max=<0 0>, min=<0 0>}
    }
    padding=<ltrb 0 0 0 0>
   }
   slot {class=ScrollBoxSlot}
  }
  VerticalBox_70 {
   widgetClassPath=&VerticalBox
   parent=Border_59
   widget=VerticalBox_70
   slot {class=BorderSlot, properties {padding=<ltrb 0 0 0 0>}}
  }
  TitleText {
   widgetClassPath=&TextBlock
   parent=VerticalBox_70
   widget=TitleText
   properties {text="Foley Clipboard             "}
   slot {class=VerticalBoxSlot}
  }
  Spacer_94 {widgetClassPath=&Spacer, parent=VerticalBox_70, widget=Spacer_94, properties {size=<1 24>}, slot {class=VerticalBoxSlot}}
  ErrorBanner {
   widgetClassPath=&TextBlock
   bIsVariable=true
   parent=VerticalBox_70
   widget=ErrorBanner
   properties {visibility=Collapsed}
   slot {class=VerticalBoxSlot}
  }
  RowsContainer {widgetClassPath=&VerticalBox, bIsVariable=true, parent=VerticalBox_70, widget=RowsContainer, slot {class=VerticalBoxSlot}}
  Spacer {widgetClassPath=&Spacer, parent=VerticalBox_70, widget=Spacer, properties {size=<1 24>}, slot {class=VerticalBoxSlot}}
  ResetButton {
   widgetClassPath=&Button
   parent=VerticalBox_70
   widget=ResetButton
   properties {widgetStyle {
     clickedSlateSound {resourceObject=null}
     disabled {
      bIsDynamicallyLoaded=false
      drawAs=RoundedBox
      imageType=NoImage
      mirroring=NoMirror
      resourceName=None
      resourceObject=null
      tiling=NoTile
      imageSize=<0 0>
      margin=<ltrb 0 0 0 0>
      outlineSettings {
       bUseBrushTransparency=false
       roundingType=FixedRadius
       width=1
       color {colorUseRule=UseColor_ColorTable, specifiedColor=<rgba 0 0 0 0>}
       cornerRadii=<4 4 4 4>
      }
      tintColor {colorUseRule=UseColor_Specified, specifiedColor=<rgba 0.0395 0.0395 0.0395 1>}
      uVRegion {bIsValid=false, max=<0 0>, min=<0 0>}
     }
     disabledForeground {colorUseRule=UseColor_ColorTable, specifiedColor=<rgba 0 0 0 0>}
     hovered {
      bIsDynamicallyLoaded=false
      drawAs=RoundedBox
      imageType=NoImage
      mirroring=NoMirror
      resourceName=None
      resourceObject=null
      tiling=NoTile
      imageSize=<32 32>
      margin=<ltrb 0 0 0 0>
      outlineSettings {
       bUseBrushTransparency=false
       roundingType=FixedRadius
       width=1
       color {colorUseRule=UseColor_ColorTable, specifiedColor=<rgba 0 0 0 0>}
       cornerRadii=<4 4 4 4>
      }
      tintColor {colorUseRule=UseColor_Specified, specifiedColor=<rgba 0.0953 0.0953 0.0953 1>}
      uVRegion {bIsValid=false, max=<0 0>, min=<0 0>}
     }
     hoveredForeground {colorUseRule=UseColor_ColorTable, specifiedColor=<rgba 0 0 0 0>}
     hoveredSlateSound {resourceObject=null}
     normal {
      bIsDynamicallyLoaded=false
      drawAs=RoundedBox
      imageType=NoImage
      mirroring=NoMirror
      resourceName=None
      resourceObject=null
      tiling=NoTile
      imageSize=<32 32>
      margin=<ltrb 0 0 0 0>
      outlineSettings {
       bUseBrushTransparency=false
       roundingType=FixedRadius
       width=1
       color {colorUseRule=UseColor_ColorTable, specifiedColor=<rgba 0 0 0 0>}
       cornerRadii=<4 4 4 4>
      }
      tintColor {colorUseRule=UseColor_Specified, specifiedColor=<rgba 0.0395 0.0395 0.0395 1>}
      uVRegion {bIsValid=false, max=<0 0>, min=<0 0>}
     }
     normalForeground {colorUseRule=UseColor_ColorTable, specifiedColor=<rgba 0 0 0 0>}
     normalPadding=<ltrb 12 1.5 12 1.5>
     pressed {
      bIsDynamicallyLoaded=false
      drawAs=RoundedBox
      imageType=NoImage
      mirroring=NoMirror
      resourceName=None
      resourceObject=null
      tiling=NoTile
      imageSize=<32 32>
      margin=<ltrb 0 0 0 0>
      outlineSettings {
       bUseBrushTransparency=false
       roundingType=FixedRadius
       width=1
       color {colorUseRule=UseColor_ColorTable, specifiedColor=<rgba 0 0 0 0>}
       cornerRadii=<4 4 4 4>
      }
      tintColor {colorUseRule=UseColor_Specified, specifiedColor=<rgba 0.0284 0.0284 0.0284 1>}
      uVRegion {bIsValid=false, max=<0 0>, min=<0 0>}
     }
     pressedForeground {colorUseRule=UseColor_ColorTable, specifiedColor=<rgba 0 0 0 0>}
     pressedPadding=<ltrb 12 2.5 12 0.5>
     pressedSlateSound {resourceObject=null}
    }}
   slot {class=VerticalBoxSlot}
  }
  ResetButtonLabel {
   widgetClassPath=&TextBlock
   parent=ResetButton
   widget=ResetButtonLabel
   properties {text=Reset to Defaults}
   slot {class=ButtonSlot}
  }
 ]
}
